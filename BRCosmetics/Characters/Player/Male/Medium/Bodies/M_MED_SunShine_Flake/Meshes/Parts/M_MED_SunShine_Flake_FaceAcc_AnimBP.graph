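# Animation Blueprint for the SunShine Flake cosmetic's face accessory, animating its hood fur on Fortnite player pawns.
# It layers the body mesh's pose from spine_04 over the attached parent's, then constrains fur master bones toward collar and fur bones.
# A Control Rig, whose pickaxe_alpha drops during melee animations, feeds five spring-driven AnimDynamics bones.
# Lower-mid dynamics and lower-master constraints are unconnected; the computed Gravity Override goes unused in the AnimGraph.
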AnimBlueprint /BRCosmetics/Characters/Player/Male/Medium/Bodies/M_MED_SunShine_Flake/Meshes/Parts/M_MED_SunShine_Flake_FaceAcc_AnimBP
target_skeleton=&/M_MED_SunShine_Flake_FaceAcc_Skeleton
graphs [
 AnimGraph {nodes [
   n1 = Anim Dynamics - Bone: dyn_hood_fur_r_lower_mid(Alpha=1, ExternalForce="0,0,0") {
    angularDampingOverride=1
    angularSpringConstant=5000
    bAngularSpring=true
    bGravityOverrideInSimSpace=true
    bOverrideAngularDamping=true
    bUseGravityOverride=true
    simulationSpace=BoneRelative
    alphaBoolBlend {blendOption=Linear}
    boundBone {boneName=dyn_hood_fur_r_lower_mid}
    componentAppliedLinearAccClamp=<1500 1500 1500>
    componentLinearAccScale=<0.4 0.4 1>
    componentLinearVelScale=<0.2 0.2 0.4>
    gravityOverride {z=-980}
    physicsBodyDefinitions [{
      sphereCollisionRadius=0
      boundBone {boneName=dyn_hood_fur_r_lower_mid}
      boxExtents=<20 20 20>
      constraintSetup {angularLimitsMax=<5 8 5>, angularLimitsMin=<-5 0 -5>, angularTarget=<1 0 0>}
      localJointOffset=<3 0 0>
     }]
    relativeSpaceBone {boneName=root}
   }
   Anim Dynamics - Bone: dyn_hood_fur_l_lower_mid(ComponentPose=@n1, Alpha=1, ExternalForce="0,0,0") {
    angularDampingOverride=1
    angularSpringConstant=5000
    bAngularSpring=true
    bGravityOverrideInSimSpace=true
    bOverrideAngularDamping=true
    bUseGravityOverride=true
    simulationSpace=BoneRelative
    alphaBoolBlend {blendOption=Linear}
    boundBone {boneName=dyn_hood_fur_l_lower_mid}
    componentAppliedLinearAccClamp=<1500 1500 1500>
    componentLinearAccScale=<0.4 0.4 1>
    componentLinearVelScale=<0.2 0.2 0.4>
    gravityOverride {z=-980}
    physicsBodyDefinitions [{
      sphereCollisionRadius=0
      boundBone {boneName=dyn_hood_fur_l_lower_mid}
      boxExtents=<20 20 20>
      constraintSetup {angularLimitsMax=<5 8 5>, angularLimitsMin=<-5 0 -5>, angularTarget=<1 0 0>}
      localJointOffset=<3 0 0>
     }]
    relativeSpaceBone {boneName=root}
   }
   n2 = "Constraint - dyn_hood_fur_r_lower_mstr (PP)"!(ComponentPose, Alpha=1, ConstraintWeights_0=0.5, ConstraintWeights_1:float=0.5) -> (Pose) {
    alphaBoolBlend {blendOption=Linear}
    boneToModify {boneName=dyn_hood_fur_r_lower_mstr}
    constraintSetup [{targetBone {boneName=dyn_collar_front_r}}, {targetBone {boneName=head}}]
    constraintWeights [1,1]
   }
   "Constraint - dyn_hood_fur_l_lower_mstr (PP)"!(ComponentPose=@n2, Alpha=1, ConstraintWeights_0=0.5, ConstraintWeights_1:float=0.5) -> (Pose) {
    alphaBoolBlend {blendOption=Linear}
    boneToModify {boneName=dyn_hood_fur_l_lower_mstr}
    constraintSetup [{targetBone {boneName=dyn_collar_front_l}}, {targetBone {boneName=head}}]
    constraintWeights [1,1]
   }
   n3 = Copy Pose From Mesh(SourceMeshComponent=Get Body_Mesh()) {bCopyCurves=true, bUseAttachedParent=true}
   n4 = Copy Pose From Mesh() {bCopyCurves=true, bUseAttachedParent=true}
   n5 = "Constraint - dyn_hood_fur_r_base_mstr (PP)"!(
    ComponentPose=Local To Component(LocalPose=Layered blend per bone(BasePose=@n4, BlendPoses_0=@n3) {layerSetup [{branchFilters [{blendDepth=50, boneName=spine_04}, {boneName=None}]}]})
    Alpha=1
    ConstraintWeights_0=1
    ConstraintWeights_1:float=0
   ) -> (Pose) {
    alphaBoolBlend {blendOption=Linear}
    boneToModify {boneName=dyn_hood_fur_r_base_mstr}
    constraintSetup [{targetBone {boneName=dyn_collar_front_r}}, {targetBone {boneName=head}}]
    constraintWeights [1,1]
   }
   n6 = "Constraint - dyn_hood_fur_l_base_mstr (PP)"!(ComponentPose=@n5, Alpha=1, ConstraintWeights_0=1, ConstraintWeights_1:float=0) -> (Pose) {
    alphaBoolBlend {blendOption=Linear}
    boneToModify {boneName=dyn_hood_fur_l_base_mstr}
    constraintSetup [{targetBone {boneName=dyn_collar_front_l}}, {targetBone {boneName=head}}]
    constraintWeights [1,1]
   }
   n7 = "Constraint - dyn_hood_fur_r_lower_mid_mstr (PP)"!(ComponentPose=@n6, Alpha=1, ConstraintWeights_0=0.5, ConstraintWeights_1:float=0.5) -> (Pose) {
    alphaBoolBlend {blendOption=Linear}
    boneToModify {boneName=dyn_hood_fur_r_lower_mid_mstr}
    constraintSetup [{targetBone {boneName=dyn_collar_front_r}}, {targetBone {boneName=dyn_hood_fur_r_lower}}]
    constraintWeights [1,1]
   }
   n8 = "Constraint - dyn_hood_fur_l_lower_mid_mstr (PP)"!(ComponentPose=@n7, Alpha=1, ConstraintWeights_0=0.5, ConstraintWeights_1:float=0.5) -> (Pose) {
    alphaBoolBlend {blendOption=Linear}
    boneToModify {boneName=dyn_hood_fur_l_lower_mid_mstr}
    constraintSetup [{targetBone {boneName=dyn_collar_front_l}}, {targetBone {boneName=dyn_hood_fur_l_lower}}]
    constraintWeights [1,1]
   }
   n9 = Control Rig!(Source=Component To Local(ComponentPose=@n8), Alpha, pickaxe_alpha:double=Get pickaxe_alpha() -> (pickaxe_alpha:double)) -> (Pose) {
    alphaBoolBlend {blendOption=Linear}
    controlRigAssetReference {blueprintRigClass=&/M_MED_SunShine_Flake_FaceAcc_CtrlRig.M_MED_SunShine_Flake_FaceAcc_CtrlRig_C}
   }
   n10 = Anim Dynamics - Bone: dyn_hood_fur_l_upper(ComponentPose=Local To Component(LocalPose=@n9), Alpha=1, ExternalForce="0,0,0") {
    angularDampingOverride=1
    angularSpringConstant=5000
    bAngularSpring=true
    bGravityOverrideInSimSpace=true
    bOverrideAngularDamping=true
    bUseGravityOverride=true
    simulationSpace=BoneRelative
    alphaBoolBlend {blendOption=Linear}
    boundBone {boneName=dyn_hood_fur_l_upper}
    componentAppliedLinearAccClamp=<1500 1500 1500>
    componentLinearAccScale=<0.4 0.4 1>
    componentLinearVelScale=<0.2 0.2 0.4>
    gravityOverride {z=-980}
    physicsBodyDefinitions [{
      sphereCollisionRadius=0
      boundBone {boneName=dyn_hood_fur_l_upper}
      boxExtents=<20 20 20>
      constraintSetup {angularLimitsMax=<3 10 10>, angularLimitsMin=<-3 0 -10>, angularTarget=<1 0 0>}
      localJointOffset=<3 0 0>
     }]
    relativeSpaceBone {boneName=root}
   }
   n11 = Anim Dynamics - Bone: dyn_hood_fur_l_lower(ComponentPose=@n10, Alpha=1, ExternalForce="0,0,0") {
    angularDampingOverride=1
    angularSpringConstant=5000
    bAngularSpring=true
    bGravityOverrideInSimSpace=true
    bOverrideAngularDamping=true
    bUseGravityOverride=true
    simulationSpace=BoneRelative
    alphaBoolBlend {blendOption=Linear}
    boundBone {boneName=dyn_hood_fur_l_lower}
    componentAppliedLinearAccClamp=<1500 1500 1500>
    componentLinearAccScale=<0.4 0.4 1>
    componentLinearVelScale=<0.2 0.2 0.4>
    gravityOverride {z=-980}
    physicsBodyDefinitions [{
      sphereCollisionRadius=0
      boundBone {boneName=dyn_hood_fur_l_lower}
      boxExtents=<20 20 20>
      constraintSetup {angularLimitsMax=<3 10 10>, angularLimitsMin=<-3 0 -10>, angularTarget=<1 0 0>}
      localJointOffset=<3 0 0>
     }]
    relativeSpaceBone {boneName=root}
   }
   n12 = Anim Dynamics - Bone: dyn_hood_fur_r_upper(ComponentPose=@n11, Alpha=1, ExternalForce="0,0,0") {
    angularDampingOverride=1
    angularSpringConstant=5000
    bAngularSpring=true
    bGravityOverrideInSimSpace=true
    bOverrideAngularDamping=true
    bUseGravityOverride=true
    simulationSpace=BoneRelative
    alphaBoolBlend {blendOption=Linear}
    boundBone {boneName=dyn_hood_fur_r_upper}
    componentAppliedLinearAccClamp=<1500 1500 1500>
    componentLinearAccScale=<0.4 0.4 1>
    componentLinearVelScale=<0.2 0.2 0.4>
    gravityOverride {z=-980}
    physicsBodyDefinitions [{
      sphereCollisionRadius=0
      boundBone {boneName=dyn_hood_fur_r_upper}
      boxExtents=<20 20 20>
      constraintSetup {angularLimitsMax=<3 10 7>, angularLimitsMin=<-3 0 -7>, angularTarget=<1 0 0>}
      localJointOffset=<3 0 0>
     }]
    relativeSpaceBone {boneName=root}
   }
   n13 = Anim Dynamics - Bone: dyn_hood_fur_r_lower(ComponentPose=@n12, Alpha=1, ExternalForce="0,0,0") {
    angularDampingOverride=1
    angularSpringConstant=5000
    bAngularSpring=true
    bGravityOverrideInSimSpace=true
    bOverrideAngularDamping=true
    bUseGravityOverride=true
    simulationSpace=BoneRelative
    alphaBoolBlend {blendOption=Linear}
    boundBone {boneName=dyn_hood_fur_r_lower}
    componentAppliedLinearAccClamp=<1500 1500 1500>
    componentLinearAccScale=<0.4 0.4 1>
    componentLinearVelScale=<0.2 0.2 0.4>
    gravityOverride {z=-980}
    physicsBodyDefinitions [{
      sphereCollisionRadius=0
      boundBone {boneName=dyn_hood_fur_r_lower}
      boxExtents=<20 20 20>
      constraintSetup {angularLimitsMax=<3 10 5>, angularLimitsMin=<-3 0 -5>, angularTarget=<1 0 0>}
      localJointOffset=<3 0 0>
     }]
    relativeSpaceBone {boneName=root}
   }
   n14 = Anim Dynamics - Bone: dyn_hood_fur_c_top(ComponentPose=@n13, Alpha=1, ExternalForce="0,0,0") {
    angularDampingOverride=1
    angularSpringConstant=5000
    bAngularSpring=true
    bGravityOverrideInSimSpace=true
    bOverrideAngularDamping=true
    bUseGravityOverride=true
    simulationSpace=BoneRelative
    alphaBoolBlend {blendOption=Linear}
    boundBone {boneName=dyn_hood_fur_c_top}
    componentAppliedLinearAccClamp=<1500 1500 1500>
    componentLinearAccScale=<0.4 0.4 1>
    componentLinearVelScale=<0.2 0.2 0.4>
    gravityOverride {z=-980}
    physicsBodyDefinitions [{
      sphereCollisionRadius=0
      boundBone {boneName=dyn_hood_fur_c_top}
      boxExtents=<20 20 20>
      constraintSetup {angularLimitsMax=<3 5 10>, angularLimitsMin=<-3 -10 -10>, angularTarget=<1 0 0>}
      localJointOffset=<3 0 0>
     }]
    relativeSpaceBone {boneName=root}
   }
   Output Pose(Result=Component To Local(ComponentPose=@n14)) {layerGroup=None}
  ]}
 EventGraph {nodes [
   Event BlueprintUpdateAnimation()
   > n1 = Gravity Override State Function DEPRECATED(CharacterPartAnimInstance=Self-Reference(), GameTimeInSeconds=GetGameTimeInSeconds(), GravityOverrideParams=Get Gravity Override Params())
   > Set Gravity Override(Gravity Override=@n1)
   > Set pickaxe_alpha(pickaxe_alpha:double=SelectFloat(A=0.5, B=1, bPickA=Get bIsPlayingMeleeAnim())) -> (Output_Get:double)
   Event BlueprintInitializeAnimation()
   > Set Skel_Mesh(Skel_Mesh=GetOwnerSkeletalMeshComponent())
   > n2 = Cast To PlayerPawn_Athena(Object=GetOwningActor())
   > n3 = Set Pawn Owner(Pawn Owner:Fort Player Pawn Athena Object Reference=@n2) -> (Output_Get:Fort Player Pawn Athena Object Reference)
   > Set Body_Mesh(Body_Mesh=GetSkeletalMeshForPartType(self=@n3))
   TryGetPawnOwner()
  ]}
]
variables [
 Body_Mesh {type=SkeletalMeshComponent}
 Gravity Override {type=Vector}
 Gravity Override Params {type=GravityOverrideParamsStruct}
 Pawn Owner {type=FortPlayerPawnAthena}
 pickaxe_alpha {type=number}
 Skel_Mesh {type=SkeletalMeshComponent}
]
